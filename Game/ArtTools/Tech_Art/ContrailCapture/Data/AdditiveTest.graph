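# AdditiveTest is a minimal additive-blend test material in the ContrailCapture tech-art data folder.
# It uses the Default Lit shading model and is flagged for use with instanced static meshes.
# Its graph holds only a constant of one, with no link to a material output shown, and has no textures or parameters.
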Material /Game/ArtTools/Tech_Art/ContrailCapture/Data/AdditiveTest
graph {
 properties {bUsedWithInstancedStaticMeshes=true, blendMode=BLEND_Additive, pixelDepthOffsetMode=PDOM_Legacy, usedShadingModels=Default Lit}
 nodes [Constant() {r=1}]
}
