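; Capture rig Blueprint for previewing skydive contrail cosmetics on a test mannequin, under an art tools ContrailCapture folder.
; Construction applies a defaults DataTable row through functions that set camera translation, orbit, FOV, look-at shift and character/emitter tilts.
; UpdateContrail blocking-loads the item's Niagara effect, sets front-end preview flags and snaps it to the mannequin's root.
; A post-process material toggles a 2D Jonesy overlay; CameraLookAtUpdate and Force 2D Position are empty.
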(event EventBeginPlay
  (Camera|CameraLookatUpdate)
  (Transformation|AttachComponentToComponent (Variables|Default|GetContrailFXSystem) (Class|ContrailTestMann|GetSkeletonComp (Utilities|Casting|CastToContrailTestMann (Variables|/Script/Engine.ChildActorComponent|GetChildActor (Variables|Default|GetContrailMann)))) "root" "SnapToTarget" "KeepRelative" "KeepRelative" true))

(event Collision|EventActorBeginOverlap
  (params (OtherActor :type Actor%20Object%20Reference)))

(event EventTick
  (params (DeltaSeconds :type Float%20%28single-precision%29)))

(event Custom|UpdateContrail
  (params (Contrail :type Athena%20Sky%20Dive%20Contrail%20Item%20Definition%20Object%20Reference))
  (bind _contrailfxsystem (Variables|Default|GetContrailFXSystem))
  (bind _output_get (Variables|Contrail|SetContrail Contrail))
  (bind _returnvalue (Utilities|LoadAssetBlocking (Variables|/Script/FortniteGame.AthenaSkyDiveContrailItemDefinition|GetNiagaraContrailEffect _output_get)))
  (bind _asniagara_system (Utilities|Casting|CastToNiagaraSystem _returnvalue)
    (:then
      (Niagara|SetNiagaraSystemAsset _contrailfxsystem _asniagara_system true)
      (Effects|Components|ParticleSystem|SetBoolParameter _contrailfxsystem "bIsFrontEnd" true)
      (Effects|Components|ParticleSystem|SetBoolParameter _contrailfxsystem "bIsFrontEndPreview" true)
      (Transformation|AttachComponentToComponent (Variables|Default|GetContrailFXSystem) (Class|ContrailTestMann|GetSkeletonComp (Utilities|Casting|CastToContrailTestMann (Variables|/Script/Engine.ChildActorComponent|GetChildActor (Variables|Default|GetContrailMann)))) "root" "SnapToTarget" "KeepRelative" "KeepRelative" true))
    (:CastFailed
      (Development|PrintString "Error loading in Rig"))))

(event Custom|Set%202D%20Overlay
  (params (Show :type Boolean))
  (bind _viewcam (Variables|Default|GetViewCam))
  ;[1 1 1] Toggle 2D Jonesy
  (if Show
    (Variables|/Script/Engine.CameraComponent|SetPostProcessSettings (Utilities|Struct|MakePostProcessSettings (Utilities|Struct|MakeWeightedBlendables (Utilities|Array|MakeArray (struct :Weight 1.0 :Object "/Game/ArtTools/Tech_Art/ContrailCapture/Data/M_JonesyOverlay.M_JonesyOverlay")))) _viewcam)
    (else
      (Variables|/Script/Engine.CameraComponent|SetPostProcessSettings (Utilities|Struct|MakePostProcessSettings (Utilities|Struct|MakeWeightedBlendables (Utilities|Array|MakeArray (struct :Weight 0.0 :Object "/Game/ArtTools/Tech_Art/ContrailCapture/Data/M_JonesyOverlay.M_JonesyOverlay")))) _viewcam))))

(event Custom|Force%202D%20Position
  (params (Force :type Boolean)))

(fn CameraZShiftHorizontalCenter
  (params (ZShift :type Float%20%28double-precision%29))
  (Transformation|SetRelativeLocation (Variables|Default|GetHorizontalRotationCenter) (Math|Vector|MakeVector 0.0 0.0 ZShift) false false))

(fn Camera_FOV
  (params (FOV :type Float%20%28double-precision%29))
  (Camera|SetFieldOfView (Variables|Default|GetViewCam) FOV))

(fn Camera_Tilt
  (params (Tilt :type Float%20%28double-precision%29))
  (Transformation|SetRelativeRotation (Variables|Default|GetViewCam) (Math|Rotator|MakeRotator 0.0 (* Tilt 6e+01) -9e+01) false false))

(fn Camera_Translate
  (params (Translation :type Float%20%28double-precision%29))
  (Transformation|SetRelativeLocation (Variables|Default|GetViewCam) (Math|Vector|MakeVector 0.0 Translation) false false)
  (Camera|CameraLookatUpdate))

(fn CameraHorizontalRotation
  (params (RotationAmount :type Float%20%28double-precision%29))
  (bind (_relativerotation_roll _relativerotation_pitch) (Math|Rotator|BreakRotator (Variables|/Script/Engine.SceneComponent|GetRelativeRotation_Roll (Variables|Default|GetHorizontalRotationCenter))))
  (Transformation|SetRelativeRotation (Variables|Default|GetHorizontalRotationCenter) (Math|Rotator|MakeRotator _relativerotation_roll _relativerotation_pitch (* RotationAmount 3.6e+02)) false false)
  (Camera|CameraLookatUpdate))

(fn CameraLookAtUpdate)

(fn CameraVerticalRotation
  (params (Theta :type Float%20%28double-precision%29))
  (Transformation|SetRelativeRotation (Variables|Default|GetVerticalRotationCenter) (Math|Rotator|MakeRotator (* Theta 1.8e+02)) false false)
  (Camera|CameraLookatUpdate))

(fn Character_Emitter_Tilt
  (params (Rotation :type Float%20%28double-precision%29))
  (Transformation|SetRelativeRotation (Variables|Default|GetContrailFXSystem) (Math|Rotator|MakeRotator (* Rotation 8e+01)) false false))

(fn Character_Tilt
  (params (Rotation :type Float%20%28double-precision%29))
  ;[1 1 1] Rotate Character
  (Transformation|SetRelativeRotation (Variables|Default|GetCharacterTiltCenter) (Math|Rotator|MakeRotator (* Rotation 6e+01)) false false))

(fn LookAtSphere_Shift
  (params (Shift :type Float%20%28double-precision%29))
  (Transformation|SetRelativeLocation (Variables|Default|GetLookAtSphere) (Math|Vector|MakeVector 0.0 0.0 (+ (* Shift 1e+02) (Variables|Default|GetLookAtSphere_DefaultLoc))) false false)
  (Camera|CameraLookatUpdate))

(fn ConstructionScript
  (bind _contrailfxsystem (Variables|Default|GetContrailFXSystem))
  (Camera|CameraLookatUpdate)
  (bind _returnvalue (Utilities|GetDataTableRow (Variables|Default|GetDefaults) "Defaults")
    (:then
      ;[1 1 1] Load Initial defaults
      (Utilities|FlowControl|Sequence
        (:then_0
          (CallFunction|CameraTranslate (Utilities|Struct|BreakStructContrailCaptureSettings _returnvalue)))
        (:then_1
          (Camera|CameraHorizontalRotation (.HorizontalRotation_7_F8624942410561E8DF656A9FEBAAE465 (Utilities|Struct|BreakStructContrailCaptureSettings _returnvalue))))
        (:then_2
          (Camera|CameraVerticalRotation (.VerticalRotation_8_F0ACF91543E90E9F6BA831ACD82F328E (Utilities|Struct|BreakStructContrailCaptureSettings _returnvalue))))
        (:then_3
          (Character|CharacterTilt (.CharacterTilt_9_66732E014F7B08A0DB9AC5B5CAED0056 (Utilities|Struct|BreakStructContrailCaptureSettings _returnvalue))))
        (:then_4
          (Character|CharacterEmitterTilt (.EmitterTilt_10_0B4EAC7D422544C867BB2B9478F68D9B (Utilities|Struct|BreakStructContrailCaptureSettings _returnvalue))))
        (:then_5
          (Camera|LookatSphereShift (.LookAtShift_12_95A044C64C3F578793A85883CA91D356 (Utilities|Struct|BreakStructContrailCaptureSettings _returnvalue))))
        (:then_6
          (CallFunction|CameraFOV (.FOV_15_1B60BC90488D5DC285F83789592948B4 (Utilities|Struct|BreakStructContrailCaptureSettings _returnvalue))))
        (:then_7
          (CallFunction|CameraZShiftHorizontalCenter (.ZShift_18_74AC7B4843CD45D256E90AA63591D9B9 (Utilities|Struct|BreakStructContrailCaptureSettings _returnvalue))))
        (:then_8
          (Effects|Components|ParticleSystem|SetBoolParameter _contrailfxsystem "bIsFrontEnd" true)
          (Effects|Components|ParticleSystem|SetBoolParameter _contrailfxsystem "bIsFrontEndPreview" true))))
    (:RowNotFound
      (Development|PrintString "Error Finding Defaults on Rig Construct"))))
