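; ContrailTestMann is a skeletal mesh mannequin actor in the contrail capture art tools, holding an assignable Athena skydive contrail item definition.
; ChangeContrail stores the new definition and calls UpdateFXSystem, whose event body is empty here.
; BeginPlay fades in the mannequin audio, or stops it if the hidden flag is set, and Destroyed fades it out.
; GetSkeletonComp exposes the skeletal mesh component.
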(event EventDestroyed
  (bind _mannequinaudio (Variables|Default|GetMannequinAudio)
    (:then
      (Audio|Components|Audio|FadeOut _mannequinaudio 0.25 0.0 "Linear" :!class /Script/Engine.AudioComponent))
    (:else)))

(event EventBeginPlay
  (bind _mannequinaudio (Variables|Default|GetMannequinAudio))
  (Variables|Default|GetMannequinAudio
    (:then
      (if (Variables|Rendering|GetbHidden)
        (Audio|Components|Audio|Stop _mannequinaudio)
        (else
          (Audio|Components|Audio|FadeIn _mannequinaudio 0.25 1.0 0.0 "Linear" :!class /Script/Engine.AudioComponent))))
    (:else)))

(event Custom|Update%20FX%20System)

(event Custom|ChangeContrail
  (params (Contrail :type Athena%20Sky%20Dive%20Contrail%20Item%20Definition%20Object%20Reference))
  (Variables|Default|SetContrail Contrail)
  (CallFunction|UpdateFXSystem))

(fn GetSkeletonComp
  (return (Variables|SkeletalMeshActor|GetSkeletalMeshComponent)))

(fn ConstructionScript)
