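# Editor utility actor Blueprint in the contrail capture tools that opens an assigned Editor Utility Widget as an editor tab.
# Its OpenControlWidget event checks an enable flag, validates the widget asset, and prints a message if none is valid.
# An inline Python node spawns and registers the tab from the asset path, returning the tab ID for storage in EUW_Name.
# Despite the asset name, no unload logic is present.
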EditorUtilityBlueprint /Game/ArtTools/Tech_Art/ContrailCapture/Data/EUB_EUW_Load_Unloader
graphs [
 EventGraph {nodes [
   OpenControlWidget() {is=title K2Node_CustomEvent}
   > Branch(Condition=Get Enable?() -> (Enable?:bool))
   n1 = Get EUW_ToSpawn() -> (EUW_ToSpawn:Editor Utility Widget Blueprint Object Reference)
   > n2 = Is Valid(InputObject=@n1)
   PrintString(execute=@n2.Is Not Valid, InString=No Valid EUW selected)
   n3 = Execute Python Script!(
    execute=@n2.Is Valid
    PythonScript="import unreal\r\n\r\nid = unreal.EditorUtilitySubsystem().spawn_and_register_tab_and_get_id(unreal.EditorAssetLibrary.load_asset(euw))\r\n\r\nprint(\"~~~~~~~~\")\r\nprint(id)\r\n\r\nid_out = id[1]"
    EUW:String=Get Object Path String(Object=@n1)
   ) -> (then, ReturnValue, id_out:Name) {inputs [EUW], outputs [id_out]}
   > Set EUW_Name(EUW_Name:Name=@n3.id_out) -> (Output_Get:Name=None)
   Event BeginPlay()
   Event ActorBeginOverlap()
   Event Tick!() -> (OutputDelegate, then, DeltaSeconds:float)
  ]}
 UserConstructionScript {nodes [Construction Script()]}
]
variables [
 Enable? {type=boolean}
 EUW_Name {type=string, default=None}
 EUW_ToSpawn {type=EditorUtilityWidgetBlueprint, default=null, class=~, animations=~, widgets=~}
]
