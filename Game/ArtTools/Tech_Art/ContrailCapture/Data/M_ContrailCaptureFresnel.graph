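# Unlit masked material for skeletal meshes in the ContrailCapture tech-art tool, computing a hard-edged directional rim.
# A dot product of pixel normal and camera-space light vector is shaped by hardness, exponent and strength.
# Pre-skinned height and a falloff for vertical view angles mask the rim before Ceil makes it binary.
# The JonesyMaterial value from MPC_ContrailRig blends a vector parameter color with that rim.
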Material /Game/ArtTools/Tech_Art/ContrailCapture/Data/M_ContrailCaptureFresnel
graph {
 properties {
  bUsedWithMeshDeformer=true
  bUsedWithSkeletalMesh=true
  blendMode=BLEND_Masked
  pixelDepthOffsetMode=PDOM_Legacy
  shadingModel=MSM_Unlit
  usedShadingModels=Unlit
 }
 nodes [
  n1 = PixelNormalWS()
  n2 = VectorParameter() {group=Parallel, parameterName=parallel_rimLightVector, sortPriority=10, defaultValue=<rgba 1 0.3 1 0>}
  n3 = Transform(None=Normalize(VectorInput=@n2.RGB)) {transformSourceType=TRANSFORMSOURCE_Camera}
  n4 = Power(
   Base=Multiply(A=DotProduct(A=@n1, B=@n3), B=ScalarParameter() {defaultValue=1, group=Parallel, parameterName=parallel_rimHardness, sortPriority=20}) {constB=10}
   Exp=ScalarParameter() {defaultValue=1, group=Parallel, parameterName=parallel_rimExponent}
  )
  n5 = MaterialFunctionCall(Float3=VertexInterpolator(VS=PreSkinnedPosition()).PS) {materialFunction=&BreakOutFloat3Components}
  n6 = Multiply(
   A=Multiply(A=Saturate(None=@n4), B=ScalarParameter() {defaultValue=0.5, group=Parallel, parameterName=parallel_rimStrength})
   B=Saturate(None=InverseLinearInterpolate(Value=@n5.B) {constA=10, constB=150})
  )
  n7 = DotProduct(A=CameraVectorWS(), B=Constant3Vector() {constant=<rgba 0 0 1 0>})
  n8 = Saturate(None=OneMinus(None=Abs(None=@n7)))
  n9 = CollectionParameter() {collection=&/MPC_ContrailRig, parameterName=JonesyMaterial}
  LinearInterpolate(A=VectorParameter().RGB, B=Ceil(None=Multiply(A=@n6, B=@n8)), Alpha=@n9)
  Saturate(None=Power(Base=Fresnel()) {constExponent=6})
  Saturate(None=Subtract(B=@n9) {constA=2})
  VertexNormalWS()
 ]
}
