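# Post-process material in the ContrailCapture tech art tools that works with a positioned Jonesy image as a screen-space overlay.
# Samples the image at viewport UVs and uses its alpha to blend scene color toward its squared RGB, or red toward green.
# Also builds a mask from a ceiled depth sphere mask, summed scene color and image alpha, with saturated and inverted variants.
# No connections to material outputs appear in the graph.
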Material /Game/ArtTools/Tech_Art/ContrailCapture/Data/M_JonesyOverlay
graph {
 properties {blendableOutputAlpha=true, materialDomain=MD_PostProcess, pixelDepthOffsetMode=PDOM_Legacy, usedShadingModels=Default Lit}
 nodes [
  Constant()
  n1 = TextureSample(UVs=ScreenPosition().ViewportUV) {samplerType=SAMPLERTYPE_LinearColor, texture=&/Jonesy-Positioned}
  LinearInterpolate(
   A=VectorParameter() {defaultValue=<rgba 1 0 0 0>}.RGB
   B=VectorParameter() {parameterName=vvvv, defaultValue=<rgba 0.0061 1 0.0562 0>}.RGB
   Alpha=@n1.A
  )
  n2 = SceneTexture() {sceneTextureId=PPI_PostProcessInput0}
  LinearInterpolate(A=ComponentMask(None=@n2.Color) {b=true, g=true, r=true}, B=Multiply(A=@n1.RGB, B=@n1.RGB), Alpha=@n1.A) {constAlpha=0.2}
  n3 = SceneTexture() {sceneTextureId=PPI_SceneDepth}
  n4 = Constant()
  n5 = Add(
   A=Ceil(None=SphereMask(A=@n3.Color, B=@n4) {attenuationRadius=10000})
   B=MaterialFunctionCall(f3=@n2.Color) {materialFunction=&AddComponents}.f3
  )
  n6 = Add(A=@n5, B=@n1.A)
  OneMinus(None=@n6)
  Saturate(None=@n5)
  Saturate(None=@n6)
  Subtract(B=@n6)
 ]
}
