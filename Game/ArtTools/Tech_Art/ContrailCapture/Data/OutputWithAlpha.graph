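# Post-process material in the contrail capture tools that derives an alpha mask for the captured scene, with blendable alpha output enabled.
# A sphere mask on scene depth, rounded up with Ceil, flags pixels within a fixed depth range.
# That mask is added to the summed components of the scene color and saturated, so bright pixels beyond that range also register.
# A separate lookup samples the unaltered post-process input.
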Material /Game/ArtTools/Tech_Art/ContrailCapture/Data/OutputWithAlpha
graph {
 properties {blendableOutputAlpha=true, materialDomain=MD_PostProcess, pixelDepthOffsetMode=PDOM_Legacy, shadingModel=MSM_Unlit, usedShadingModels=Unlit}
 nodes [
  n1 = SphereMask(A=SceneTexture() {sceneTextureId=PPI_SceneDepth}.Color, B=Constant()) {attenuationRadius=10000}
  n2 = SceneTexture() {sceneTextureId=PPI_PostProcessInput0}
  Saturate(None=Add(A=Ceil(None=@n1), B=MaterialFunctionCall(f3=@n2.Color) {materialFunction=&AddComponents}.f3))
  SceneTexture() {sceneTextureId=PPI_PostProcessInput0}
 ]
}
